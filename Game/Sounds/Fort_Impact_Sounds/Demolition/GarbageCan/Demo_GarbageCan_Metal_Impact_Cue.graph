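# Sound cue for metal garbage can impacts, filed under the demolition impact sounds.
# A random node picks one of three equally weighted impact waves, then a modulator varies its pitch slightly for variety.
# It plays in the Fort_Impacts sound class at reduced volume with small-object demolition hit attenuation.
# An object-impact concurrency set, named for a four-sound limit, governs overlapping playback.
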SoundCue /Game/Sounds/Fort_Impact_Sounds/Demolition/GarbageCan/Demo_GarbageCan_Metal_Impact_Cue
root=Modulator_10
nodes [
 {id=Modulator_10, children [Random_23], properties {pitchMax=1.07, pitchMin=0.92}}
 {id=Random_23, children [WavePlayer_82, WavePlayer_83, WavePlayer_84], properties {weights [1,1,1]}}
 {id=WavePlayer_82, properties {soundWaveAssetPtr=&/GarbageCan_Metal_Impact01}}
 {id=WavePlayer_83, properties {soundWaveAssetPtr=&/GarbageCan_Metal_Impact02}}
 {id=WavePlayer_84, properties {soundWaveAssetPtr=&/GarbageCan_Metal_Impact03}}
]
properties {
 attenuationSettings=/Game/Sounds/Attenuation/Fort_Demolition_Hit_Small_Sized_Object_Att
 maxDistance=4000
 soundClassObject=&Fort_Impacts
 volumeMultiplier=0.34
 concurrencyOverrides {resolutionRule=PreventNew}
 concurrencySet [&CONC_Object_Impacts_FourMax]
}
