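# Sound cue for the impact of a large plastic garbage can, part of the demolition impact sounds.
# A Random node picks one of three equally weighted impact waves and feeds it through a Modulator that slightly varies pitch.
# It plays in the impacts sound class with small-object demolition hit attenuation and uses a shared concurrency set for object impacts, named for a four-sound cap.
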SoundCue /Game/Sounds/Fort_Impact_Sounds/Demolition/GarbageCan/Demo_GarbageCan_Plastic_Large_Impact_Cue
root=Modulator_10
nodes [
 {id=Modulator_10, children [Random_13], properties {pitchMax=1.07, pitchMin=0.92}}
 {id=Random_13, children [WavePlayer_169, WavePlayer_170, WavePlayer_171], properties {weights [1,1,1]}}
 {id=WavePlayer_169, properties {soundWaveAssetPtr=&/GarbageCan_Plastic_Large_Impact03}}
 {id=WavePlayer_170, properties {soundWaveAssetPtr=&/GarbageCan_Plastic_Large_Impact04}}
 {id=WavePlayer_171, properties {soundWaveAssetPtr=&/GarbageCan_Plastic_Large_Impact05}}
]
properties {
 attenuationSettings=/Game/Sounds/Attenuation/Fort_Demolition_Hit_Small_Sized_Object_Att
 maxDistance=4000
 soundClassObject=&Fort_Impacts
 volumeMultiplier=0.68
 concurrencyOverrides {resolutionRule=PreventNew}
 concurrencySet [&CONC_Object_Impacts_FourMax]
}
